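# StateTree for minor Spire sub-scenario B, an essence collection objective that grants a playspace quest.
# Four chained states fling essence pickups upward, querying the Spire actor and tracking pickups for destruction on scenario completion.
# An inventory-add verb detects each collection and increments a counter, looping to the last spawn state to respawn one until the target parameter is reached.
# It then sends the shield-interact custom verb per player and succeeds.
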StateTree /SpiresGameplay/Scenarios/Spire_Minor/Master_Scenario/SubScenarios/Spire_Minor_SubScenario_B/ST_SubScenario_B
global_tasks [
 {
  id=#1
  description="Get Variable [Int]"
  properties {
   iD=#2
   instance {
    _structType=/Script/EncountersRuntime.EncounterGetVariableIntStateTreeTaskInstanceData
    userActor=null
    value=0
    variableIdentifier {tagName=QuestEncounter.Spire_Minor_Combat_1.Encounter.Variable.EnemyCount}
   }
   node {
    name=None
    _structType=/Script/EncountersRuntime.EncounterGetVariableIntStateTreeTask
    bCanEditConsideredForCompletion=true
    bConsideredForCompletion=true
    bHasTaskCompletionDelegateDispatcher=false
    bTaskEnabled=true
    transitionHandlingPriority=Normal
    bindingsBatch {value=65535}
    executionRuntimeTemplateIndex {value=65535}
    instanceDataHandle {index=65535, source=None, stateHandle {index=65535}}
    instanceTemplateIndex {value=65535}
    outputBindingsBatch {value=65535}
   }
  }
 }
 {
  id=#3
  description="Get Variable [Item Definition]"
  properties {
   iD=#4
   instance {
    _structType=/Script/EncountersRuntime.EncounterGetVariableItemDefinitionStateTreeTaskInstanceData
    userActor=null
    value=null
    variableIdentifier {tagName=QuestEncounter.Spire_SubScenario.Essence}
   }
   node {
    name=None
    _structType=/Script/EncountersRuntime.EncounterGetVariableItemDefinitionStateTreeTask
    bCanEditConsideredForCompletion=true
    bConsideredForCompletion=true
    bHasTaskCompletionDelegateDispatcher=false
    bTaskEnabled=true
    transitionHandlingPriority=Normal
    bindingsBatch {value=65535}
    executionRuntimeTemplateIndex {value=65535}
    instanceDataHandle {index=65535, source=None, stateHandle {index=65535}}
    instanceTemplateIndex {value=65535}
    outputBindingsBatch {value=65535}
   }
  }
 }
]
states [
 Root {
  id=#5
  type=STATE
  children [
   #6
   #7
   #8
   #9
   #10
   #11
   #12
   #13
   #14
  ]
  tasks [{
    id=#15
    description="(Fortnite) Arithmetic: Make Int"
    properties {
     iD=#16
     instance {_structType=/Script/FortniteAIServer.FortAthenaMakeIntVariableStateTreeTaskInstanceData, defaultIntValue=0, outInt {int=0}}
     node {
      name=None
      _structType=/Script/FortniteAIServer.FortAthenaMakeIntVariableStateTreeTask
      bCanEditConsideredForCompletion=true
      bConsideredForCompletion=true
      bHasTaskCompletionDelegateDispatcher=false
      bResetOnReselect=true
      bTaskEnabled=true
      transitionHandlingPriority=Normal
      bindingsBatch {value=65535}
      executionRuntimeTemplateIndex {value=65535}
      instanceDataHandle {index=65535, source=None, stateHandle {index=65535}}
      instanceTemplateIndex {value=65535}
      outputBindingsBatch {value=65535}
     }
    }
   }]
  transitions []
 }
 Init {
  id=#6
  type=STATE
  tasks [{
    id=#17
    description=BP SDA STTask Complete State
    properties {
     iD=#18
     instanceObject=&self:StateTreeEditorData_0.StateTreeState_1.BP_SDA_STTask_CompleteState_C_0
     node {
      name=None
      _structType=&StateTreeBlueprintTaskWrapper
      taskClass=/StateDrivenActor/Tasks/BP_SDA_STTask_CompleteState.BP_SDA_STTask_CompleteState_C
     }
    }
   }]
  transitions [{id=#19, trigger=ON_STATE_COMPLETED, state {id=#20, name=None, linkType=NEXT_STATE}}]
 }
 Grant Quest {
  id=#7
  type=STATE
  tasks [
   {
    id=#21
    description=Grant Quest
    properties {
     iD=#22
     instance {
      _structType=/Script/EncountersRuntime.GrantQuestStateTreeTaskInstanceData
      grantType=GrantToPlayspace
      questToGrant=/SpiresGameplay/Scenarios/Spire_Minor/Master_Scenario/SubScenarios/Spire_Minor_SubScenario_B/Quests/PlaySpace_EncounterQuest_Spire_SubScenario_B_01
      targetActor=null
      userActor=null
     }
     node {
      name=None
      _structType=/Script/EncountersRuntime.GrantQuestStateTreeTask
      bCanEditConsideredForCompletion=true
      bConsideredForCompletion=true
      bHasTaskCompletionDelegateDispatcher=false
      bTaskEnabled=true
      transitionHandlingPriority=Normal
      bindingsBatch {value=65535}
      executionRuntimeTemplateIndex {value=65535}
      instanceDataHandle {index=65535, source=None, stateHandle {index=65535}}
      instanceTemplateIndex {value=65535}
      outputBindingsBatch {value=65535}
     }
    }
   }
   {
    id=#23
    description=BP SDA STTask Complete State
    properties {
     iD=#24
     instanceObject=&self:StateTreeEditorData_0.StateTreeState_2.BP_SDA_STTask_CompleteState_C_0
     node {
      name=None
      _structType=&StateTreeBlueprintTaskWrapper
      taskClass=&BP_SDA_STTask_CompleteState_C
     }
    }
   }
  ]
  transitions [{id=#25, trigger=ON_STATE_COMPLETED, state {id=#26, name=Root, linkType=NEXT_STATE}}]
 }
 Spawn Essence 1 {
  id=#8
  type=STATE
  tasks [
   {
    id=#27
    description=Get Scenario Actor
    properties {
     iD=#28
     instance {
      _structType=&GetEncounterActorStateTreeTaskInstanceData
      actorQuery {
       autoDescription=" ANY( QuestEncounter.Spire_Minor_Random.Encounter.Target.Spire )"
       tagDictionary [{tagName=QuestEncounter.Spire_Minor_Random.Encounter.Target.Spire}]
      }
     }
     node {name=None, _structType=&GetEncounterActorStateTreeTask}
    }
   }
   {
    id=#29
    description=Spawn Pickup
    properties {
     iD=#30
     instance {
      _structType=/Script/FortniteGame.SpawnPickupStateTreeTaskInstanceData
      instigatorActor=null
      itemDefinition=/SpiresGameplay/Scenarios/Spire_Minor/Master_Scenario/SubScenarios/Spire_Minor_SubScenario_B/Items/DA_Essence_Pickup
      lootTierName=None
      spawnedPickup=null
      targetActor=null
      userActor=null
      flingConeAngle {registryType=None, value=100, curve {curveTable=null, rowName=None}}
      flingDirection=<0 0 1>
      flingMagnitude {registryType=None, value=2300, curve {curveTable=null, rowName=None}}
      spawnOffset=<0 0 100>
      spawnedPickupActors []
     }
     node {
      name=None
      _structType=/Script/FortniteGame.SpawnPickupStateTreeTask
      bCanEditConsideredForCompletion=true
      bConsideredForCompletion=true
      bHasTaskCompletionDelegateDispatcher=false
      bTaskEnabled=true
      transitionHandlingPriority=Normal
      bindingsBatch {value=65535}
      executionRuntimeTemplateIndex {value=65535}
      instanceDataHandle {index=65535, source=None, stateHandle {index=65535}}
      instanceTemplateIndex {value=65535}
      outputBindingsBatch {value=65535}
     }
    }
   }
   {
    id=#31
    description=Track Scenario Actors
    properties {
     iD=#32
     instance {
      _structType=/Script/EncountersRuntime.TrackScenarioActorsStateTreeTaskInstanceData
      bDestroyOnScenarioComplete=true
      userActor=null
      actorIdentifierTags {gameplayTags []}
      targetActors []
     }
     node {
      name=None
      _structType=/Script/EncountersRuntime.TrackScenarioActorsStateTreeTask
      bCanEditConsideredForCompletion=true
      bConsideredForCompletion=true
      bHasTaskCompletionDelegateDispatcher=false
      bTaskEnabled=true
      transitionHandlingPriority=Normal
      bindingsBatch {value=65535}
      executionRuntimeTemplateIndex {value=65535}
      instanceDataHandle {index=65535, source=None, stateHandle {index=65535}}
      instanceTemplateIndex {value=65535}
      outputBindingsBatch {value=65535}
     }
    }
   }
   {
    id=#33
    description=BP SDA STTask Complete State
    properties {
     iD=#34
     instanceObject=&self:StateTreeEditorData_0.StateTreeState_3.BP_SDA_STTask_CompleteState_C_0
     node {
      name=None
      _structType=&StateTreeBlueprintTaskWrapper
      taskClass=&BP_SDA_STTask_CompleteState_C
     }
    }
   }
  ]
  transitions [{id=#35, trigger=ON_STATE_COMPLETED, state {id=#20, name=Root, linkType=NEXT_STATE}}]
 }
 Spawn Essence 2 {
  id=#9
  type=STATE
  tasks [
   {
    id=#36
    description=Get Scenario Actor
    properties {
     iD=#37
     instance {
      _structType=&GetEncounterActorStateTreeTaskInstanceData
      actorQuery {
       autoDescription=" ANY( QuestEncounter.Spire_Minor_Random.Encounter.Target.Spire )"
       tagDictionary [{tagName=QuestEncounter.Spire_Minor_Random.Encounter.Target.Spire}]
      }
     }
     node {name=None, _structType=&GetEncounterActorStateTreeTask}
    }
   }
   {
    id=#38
    description=Spawn Pickup
    properties {
     iD=#39
     instance {
      _structType=&SpawnPickupStateTreeTaskInstanceData
      instigatorActor=null
      itemDefinition=&DA_Essence_Pickup
      lootTierName=None
      spawnedPickup=null
      targetActor=null
      userActor=null
      flingConeAngle {registryType=None, value=100, curve {curveTable=null, rowName=None}}
      flingDirection=<0 0 1>
      flingMagnitude {registryType=None, value=2300, curve {curveTable=null, rowName=None}}
      spawnOffset=<0 0 100>
      spawnedPickupActors []
     }
     node {
      name=None
      _structType=&SpawnPickupStateTreeTask
      bCanEditConsideredForCompletion=true
      bConsideredForCompletion=true
      bHasTaskCompletionDelegateDispatcher=false
      bTaskEnabled=true
      transitionHandlingPriority=Normal
      bindingsBatch {value=65535}
      executionRuntimeTemplateIndex {value=65535}
      instanceDataHandle {index=65535, source=None, stateHandle {index=65535}}
      instanceTemplateIndex {value=65535}
      outputBindingsBatch {value=65535}
     }
    }
   }
   {
    id=#40
    description=Track Scenario Actors
    properties {
     iD=#41
     instance {
      _structType=&TrackScenarioActorsStateTreeTaskInstanceData
      bDestroyOnScenarioComplete=true
      userActor=null
      actorIdentifierTags {gameplayTags []}
      targetActors []
     }
     node {
      name=None
      _structType=&TrackScenarioActorsStateTreeTask
      bCanEditConsideredForCompletion=true
      bConsideredForCompletion=true
      bHasTaskCompletionDelegateDispatcher=false
      bTaskEnabled=true
      transitionHandlingPriority=Normal
      bindingsBatch {value=65535}
      executionRuntimeTemplateIndex {value=65535}
      instanceDataHandle {index=65535, source=None, stateHandle {index=65535}}
      instanceTemplateIndex {value=65535}
      outputBindingsBatch {value=65535}
     }
    }
   }
   {
    id=#42
    description=BP SDA STTask Complete State
    properties {
     iD=#43
     instanceObject=&self:StateTreeEditorData_0.StateTreeState_8.BP_SDA_STTask_CompleteState_C_0
     node {
      name=None
      _structType=&StateTreeBlueprintTaskWrapper
      taskClass=&BP_SDA_STTask_CompleteState_C
     }
    }
   }
  ]
  transitions [{id=#44, trigger=ON_STATE_COMPLETED, state {id=#45, name=Root, linkType=NEXT_STATE}}]
 }
 Spawn Essence 3 {
  id=#10
  type=STATE
  tasks [
   {
    id=#46
    description=Get Scenario Actor
    properties {
     iD=#47
     instance {
      _structType=&GetEncounterActorStateTreeTaskInstanceData
      actorQuery {
       autoDescription=" ANY( QuestEncounter.Spire_Minor_Random.Encounter.Target.Spire )"
       tagDictionary [{tagName=QuestEncounter.Spire_Minor_Random.Encounter.Target.Spire}]
      }
     }
     node {name=None, _structType=&GetEncounterActorStateTreeTask}
    }
   }
   {
    id=#48
    description=Spawn Pickup
    properties {
     iD=#49
     instance {
      _structType=&SpawnPickupStateTreeTaskInstanceData
      instigatorActor=null
      itemDefinition=&DA_Essence_Pickup
      lootTierName=None
      spawnedPickup=null
      targetActor=null
      userActor=null
      flingConeAngle {registryType=None, value=100, curve {curveTable=null, rowName=None}}
      flingDirection=<0 0 1>
      flingMagnitude {registryType=None, value=2300, curve {curveTable=null, rowName=None}}
      spawnOffset=<0 0 100>
      spawnedPickupActors []
     }
     node {
      name=None
      _structType=&SpawnPickupStateTreeTask
      bCanEditConsideredForCompletion=true
      bConsideredForCompletion=true
      bHasTaskCompletionDelegateDispatcher=false
      bTaskEnabled=true
      transitionHandlingPriority=Normal
      bindingsBatch {value=65535}
      executionRuntimeTemplateIndex {value=65535}
      instanceDataHandle {index=65535, source=None, stateHandle {index=65535}}
      instanceTemplateIndex {value=65535}
      outputBindingsBatch {value=65535}
     }
    }
   }
   {
    id=#50
    description=BP SDA STTask Complete State
    properties {
     iD=#51
     instanceObject=&self:StateTreeEditorData_0.StateTreeState_4.BP_SDA_STTask_CompleteState_C_0
     node {
      name=None
      _structType=&StateTreeBlueprintTaskWrapper
      taskClass=&BP_SDA_STTask_CompleteState_C
     }
    }
   }
   {
    id=#52
    description=Track Scenario Actors
    properties {
     iD=#53
     instance {
      _structType=&TrackScenarioActorsStateTreeTaskInstanceData
      bDestroyOnScenarioComplete=true
      userActor=null
      actorIdentifierTags {gameplayTags []}
      targetActors []
     }
     node {
      name=None
      _structType=&TrackScenarioActorsStateTreeTask
      bCanEditConsideredForCompletion=true
      bConsideredForCompletion=true
      bHasTaskCompletionDelegateDispatcher=false
      bTaskEnabled=true
      transitionHandlingPriority=Normal
      bindingsBatch {value=65535}
      executionRuntimeTemplateIndex {value=65535}
      instanceDataHandle {index=65535, source=None, stateHandle {index=65535}}
      instanceTemplateIndex {value=65535}
      outputBindingsBatch {value=65535}
     }
    }
   }
  ]
  transitions [{id=#54, trigger=ON_STATE_COMPLETED, state {id=#55, name=Root, linkType=NEXT_STATE}}]
 }
 Spawn Essence / Respawn Essence {
  id=#11
  type=STATE
  tasks [
   {
    id=#56
    description=Get Scenario Actor
    properties {
     iD=#57
     instance {
      _structType=&GetEncounterActorStateTreeTaskInstanceData
      actorQuery {
       autoDescription=" ANY( QuestEncounter.Spire_Minor_Random.Encounter.Target.Spire )"
       tagDictionary [{tagName=QuestEncounter.Spire_Minor_Random.Encounter.Target.Spire}]
      }
     }
     node {name=None, _structType=&GetEncounterActorStateTreeTask}
    }
   }
   {
    id=#58
    description=Spawn Pickup
    properties {
     iD=#59
     instance {
      _structType=&SpawnPickupStateTreeTaskInstanceData
      instigatorActor=null
      itemDefinition=&DA_Essence_Pickup
      lootTierName=None
      spawnedPickup=null
      targetActor=null
      userActor=null
      flingConeAngle {registryType=None, value=100, curve {curveTable=null, rowName=None}}
      flingDirection=<0 0 1>
      flingMagnitude {registryType=None, value=2300, curve {curveTable=null, rowName=None}}
      spawnOffset=<0 0 100>
      spawnedPickupActors []
     }
     node {
      name=None
      _structType=&SpawnPickupStateTreeTask
      bCanEditConsideredForCompletion=true
      bConsideredForCompletion=true
      bHasTaskCompletionDelegateDispatcher=false
      bTaskEnabled=true
      transitionHandlingPriority=Normal
      bindingsBatch {value=65535}
      executionRuntimeTemplateIndex {value=65535}
      instanceDataHandle {index=65535, source=None, stateHandle {index=65535}}
      instanceTemplateIndex {value=65535}
      outputBindingsBatch {value=65535}
     }
    }
   }
   {
    id=#60
    description=BP SDA STTask Complete State
    properties {
     iD=#61
     instanceObject=&self:StateTreeEditorData_0.StateTreeState_5.BP_SDA_STTask_CompleteState_C_0
     node {
      name=None
      _structType=&StateTreeBlueprintTaskWrapper
      taskClass=&BP_SDA_STTask_CompleteState_C
     }
    }
   }
   {
    id=#62
    description=Track Scenario Actors
    properties {
     iD=#63
     instance {
      _structType=&TrackScenarioActorsStateTreeTaskInstanceData
      bDestroyOnScenarioComplete=true
      userActor=null
      actorIdentifierTags {gameplayTags []}
      targetActors []
     }
     node {
      name=None
      _structType=&TrackScenarioActorsStateTreeTask
      bCanEditConsideredForCompletion=true
      bConsideredForCompletion=true
      bHasTaskCompletionDelegateDispatcher=false
      bTaskEnabled=true
      transitionHandlingPriority=Normal
      bindingsBatch {value=65535}
      executionRuntimeTemplateIndex {value=65535}
      instanceDataHandle {index=65535, source=None, stateHandle {index=65535}}
      instanceTemplateIndex {value=65535}
      outputBindingsBatch {value=65535}
     }
    }
   }
  ]
  transitions [{id=#64, trigger=ON_STATE_COMPLETED, state {id=#65, name=Root, linkType=NEXT_STATE}}]
 }
 Detect Essence Collected {
  id=#12
  type=STATE
  tasks [{
    id=#66
    description=Scenario Verb Transition
    properties {
     iD=#67
     instance {
      _structType=/Script/EncountersRuntime.EncounterVerbTransitionStateTreeTaskInstanceData
      bIsTriggered=false
      bUseAdvancedObjectives=false
      objectiveValueThreshold=1
      userActor=null
      objectiveVerbs [{
        _structType=/Script/FortniteGame.FortObjectiveFilter_AddToInventory
        acceptedGatherType=PreviouslyGathered
        bAllowItemsPreviouslyGatheredByPlayers=false
        bIncrementByItemCount=true
        bRequireBrandNewItem=false
        author {
         bCanGainProgressForThisObjective=true
         tagQuery {autoDescription="", tokenStreamVersion=0, userDescription="", queryTokenStream [], tagDictionary []}
        }
        context {tagQuery {autoDescription="", tokenStreamVersion=0, userDescription="", queryTokenStream [], tagDictionary []}}
        item {tagQuery {
          autoDescription=" ANY( QuestEncounter.Spire_SubScenario.Essence )"
          tokenStreamVersion=0
          userDescription=""
          queryTokenStream [0,1,1,1,0]
          tagDictionary [{tagName=QuestEncounter.Spire_SubScenario.Essence}]
         }}
        source {tagQuery {autoDescription="", tokenStreamVersion=0, userDescription="", queryTokenStream [], tagDictionary []}}
       }]
      objectives []
      onTriggered {iD=#68}
     }
     node {
      name=None
      _structType=/Script/EncountersRuntime.EncounterVerbTransitionStateTreeTask
      bCanEditConsideredForCompletion=true
      bConsideredForCompletion=true
      bHasTaskCompletionDelegateDispatcher=false
      bTaskEnabled=true
      transitionHandlingPriority=Normal
      bindingsBatch {value=65535}
      executionRuntimeTemplateIndex {value=65535}
      instanceDataHandle {index=65535, source=None, stateHandle {index=65535}}
      instanceTemplateIndex {value=65535}
      outputBindingsBatch {value=65535}
      transitionTo {name=None, fallback=None, iD=#68, linkType=NextState, stateHandle {index=65535}}
     }
    }
   }]
  transitions []
 }
 Collected {
  id=#13
  type=STATE
  tasks [
   {
    id=#69
    description="(Fortnite) Arithmetic: Operation Int"
    properties {
     iD=#70
     instance {_structType=/Script/FortniteAIServer.FortAthenaIntArithmeticStateTreeTaskInstanceData, firstOperand=0, referencedResult="", secondOperand=1}
     node {
      name=None
      _structType=/Script/FortniteAIServer.FortAthenaIntArithmeticStateTreeTask
      bCanEditConsideredForCompletion=true
      bConsideredForCompletion=true
      bHasTaskCompletionDelegateDispatcher=false
      bTaskEnabled=true
      operation=Add
      operationTrigger=OnEnterState
      transitionHandlingPriority=Normal
      bindingsBatch {value=65535}
      executionRuntimeTemplateIndex {value=65535}
      instanceDataHandle {index=65535, source=None, stateHandle {index=65535}}
      instanceTemplateIndex {value=65535}
      outputBindingsBatch {value=65535}
     }
    }
   }
   {
    id=#71
    description=BP SDA STTask Complete State
    properties {
     iD=#72
     instanceObject=&self:StateTreeEditorData_0.StateTreeState_7.BP_SDA_STTask_CompleteState_C_0
     node {
      name=None
      _structType=&StateTreeBlueprintTaskWrapper
      taskClass=&BP_SDA_STTask_CompleteState_C
     }
    }
   }
  ]
  transitions [
   {
    id=#73
    trigger=ON_STATE_COMPLETED
    conditions [{
      id=#74
      description="(Fortnite) Arithmetic: Make Int.Out Int.Int < Parameters.Essence Collected Target"
      properties {
       iD=#75
       instance {_structType=/Script/StateTreeModule.StateTreeCompareIntConditionInstanceData, left=0, right=0}
       node {
        name=None
        _structType=/Script/StateTreeModule.StateTreeCompareIntCondition
        bInvert=false
        deltaIndent=0
        evaluationMode=Evaluated
        operand=And
        operator=Less
        bindingsBatch {value=65535}
        executionRuntimeTemplateIndex {value=65535}
        instanceDataHandle {index=65535, source=None, stateHandle {index=65535}}
        instanceTemplateIndex {value=65535}
        outputBindingsBatch {value=65535}
       }
      }
     }]
    state {id=#11, name=Spawn Essence / Respawn Essence}
   }
   {id=#76, trigger=ON_STATE_COMPLETED, state {id=#20, name=Root, linkType=NEXT_STATE}}
  ]
 }
 Completed {
  id=#14
  type=STATE
  tasks [
   {
    id=#77
    description=Send Custom Verb
    properties {
     iD=#78
     instance {
      _structType=/Script/EncountersRuntime.EncounterSendCustomVerbStateTreeTaskInstanceData
      bSendVerbForEachPlayerAuthor=true
      userActor=null
      customVerb {amount=1, customVerbTag {tagName=CustomVerb.Minor_Spire.Shield.Interact}, subjects [], values []}
     }
     node {
      name=None
      _structType=/Script/EncountersRuntime.EncounterSendCustomVerbStateTreeTask
      bCanEditConsideredForCompletion=true
      bConsideredForCompletion=true
      bHasTaskCompletionDelegateDispatcher=false
      bTaskEnabled=true
      transitionHandlingPriority=Normal
      bindingsBatch {value=65535}
      executionRuntimeTemplateIndex {value=65535}
      instanceDataHandle {index=65535, source=None, stateHandle {index=65535}}
      instanceTemplateIndex {value=65535}
      outputBindingsBatch {value=65535}
     }
    }
   }
   {
    id=#79
    description=BP SDA STTask Complete State
    properties {
     iD=#80
     instanceObject=&self:StateTreeEditorData_0.StateTreeState_9.BP_SDA_STTask_CompleteState_C_0
     node {
      name=None
      _structType=&StateTreeBlueprintTaskWrapper
      taskClass=&BP_SDA_STTask_CompleteState_C
     }
    }
   }
  ]
  transitions [{id=#81, trigger=ON_STATE_COMPLETED, state {id=#20, name=Root, linkType=SUCCEEDED}}]
 }
]
